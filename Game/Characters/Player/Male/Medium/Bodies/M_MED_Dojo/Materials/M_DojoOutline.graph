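# Translucent unlit material that draws screen-space cartoon ink lines for the Dojo character on a shell pushed along vertex normals.
# Four-neighbor scene depth edge detection, optionally combined with GBuffer roughness edges, sets opacity; line width follows object screen size, FOV and resolution.
# GBuffer subsurface and opacity values mask lines on deferred; a pixel-versus-scene depth test replaces them on forward and mobile.
# Color follows a time-of-day curve or darkened scene diffuse.
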Material /Game/Characters/Player/Male/Medium/Bodies/M_MED_Dojo/Materials/M_DojoOutline
graph {
 properties {
  bEnableResponsiveAA=true
  bUseMaterialAttributes=true
  bUsedWithMeshDeformer=true
  bUsedWithMorphTargets=true
  bUsedWithSkeletalMesh=true
  blendMode=BLEND_Translucent
  pixelDepthOffsetMode=PDOM_Legacy
  shadingModel=MSM_Unlit
  twoSided=true
  usedShadingModels=Unlit
 }
 nodes [
  Constant() {desc=for PS4 and XO, r=1}
  n1 = VectorParameter() {parameterName=OutlineColorNight, defaultValue=<rgba 0 0 0 1>}
  n2 = VectorParameter() {parameterName=OutlineColor, defaultValue=<rgba 0 0 0 1>}
  n3 = CurveAtlasRowParameter(CurveTime=CollectionParameter() {collection=&FortniteMaterialParameters, parameterName=TimeOfDay_0-1}) {
   atlas=/Game/Characters/Shared/CurveAtlas/AthenaCharacterCurveAtlas
   curve=/Game/Characters/Player/Male/Medium/Bodies/M_MED_Dojo/Textures/DojoTOD_Curve
   defaultValue=19
   parameterName=OutlineTODCurve
  }
  n4 = ComponentMask(None=SceneTexture() {sceneTextureId=PPI_DiffuseColor}.Color) {b=true, g=true, r=true}
  n5 = VectorParameter() {parameterName=OutlineDakenMultiplier, defaultValue=<rgba 0.1 0.1 0.1 0>}
  n6 = LinearInterpolate(A=VectorParameter() {parameterName=MobileOutlineColorNight}.RGB, B=VectorParameter() {parameterName=MobileOutlineColor}.RGB, Alpha=@n3.G)
  n7 = ShadingPathSwitch(
   Default=StaticSwitchParameter(True=LinearInterpolate(A=@n1.RGB, B=@n2.RGB, Alpha=@n3.G), False=Multiply(A=@n4, B=@n5.RGB)) {parameterName=UseOutlineColor}
   Forward=@n6
   Mobile=@n6
  )
  n8 = SceneTexture() {sceneTextureId=PPI_SubsurfaceColor}
  n9 = ComponentMask(None=@n8.Color) {g=true}
  n10 = ComponentMask(None=@n8.Color) {b=true}
  n11 = ScalarParameter() {defaultValue=0.001, parameterName=SSS_Mask}
  n12 = Constant()
  n13 = Round(None=ComponentMask(None=SceneTexture() {sceneTextureId=PPI_Opacity}.Color) {r=true})
  n14 = Multiply(A=If(A=Min(A=@n9, B=@n10), B=SquareRoot(None=@n11), "A > B"=@n12, "A == B"=Constant() {r=1}, "A < B"=@n12) {equalsThreshold=0.001}, B=@n13)
  n15 = QualitySwitch(Default=@n14, Low=@n13, High=@n13, Medium=@n13, Epic=@n14)
  n16 = PixelDepth()
  n17 = SceneDepth()
  n18 = Saturate(None=Add(A=Subtract(A=@n16.R, B=@n17.R), B=ScalarParameter() {defaultValue=3, group=Cartoon, parameterName=CRTNO_Outline-InnerMaskBias}))
  n19 = ShadingPathSwitch(Default=@n15, Deferred=@n15, Forward=@n18, Mobile=@n18)
  n20 = PixelDepth()
  n21 = ScalarParameter() {defaultValue=50, group=Cartoon, parameterName=CRTNO_InnerLine_FadeStart}
  n22 = Saturate(None=Divide(A=Subtract(A=@n20.R, B=@n21), B=ScalarParameter() {defaultValue=500, group=Cartoon, parameterName=CRTNO_InnerLine_FadeDistance}))
  n23 = LinearInterpolate(
   A=ScalarParameter() {defaultValue=16, parameterName=SampleBias_Min}
   B=ScalarParameter() {defaultValue=16, parameterName=SampleBias}
   Alpha=@n22
  )
  n24 = LinearInterpolate(
   A=ScalarParameter() {defaultValue=1, parameterName=MobileSampleBias_Min}
   B=ScalarParameter() {defaultValue=16, parameterName=Mobile_SampleBias}
   Alpha=@n22
  )
  n25 = ComponentMask(None=ScreenPosition().ViewportUV) {g=true, r=true}
  n26 = Subtract(A=Divide(A=ViewSize(), B=Constant() {r=320}), B=Constant() {r=2}) {constB=0}
  n27 = Divide(A=ObjectRadius(), B=Distance(A=ObjectPositionWS(), B=CameraPositionWS()))
  n28 = ViewSize()
  n29 = Divide(
   A=Divide(A=ComponentMask(None=@n28) {r=true}, B=ComponentMask(None=@n28) {g=true})
   B=MaterialFunctionCall() {materialFunction=/Engine/Functions/Engine_MaterialFunctions02/Utility/FOV}."FOV "
  )
  n30 = Multiply(
   A=ScalarParameter() {defaultValue=2, group=Cartoon, parameterName=CRTNO_Outline-InnerSize1}
   B=Power(Base=Multiply(A=@n27, B=@n29)) {constExponent=0.6}
  )
  n31 = Multiply(A=@n30, B=ScalarParameter() {defaultValue=1.5, group=Cartoon, parameterName=CRTNO_Outline-InnerSizeMobileMultiplier})
  n32 = ShadingPathSwitch(Default=@n30, Forward=@n31, Mobile=@n31)
  n33 = Constant() {r=960}
  n34 = ComponentMask(None=ViewSize()) {r=true}
  n35 = Saturate(None=Floor(None=Divide(A=@n33, B=@n34))) {desc=if resolution is smaller than 960H}
  n36 = Multiply(
   A=Clamp(None=Ceil(None=@n26)) {maxDefault=2, minDefault=1}
   B=Multiply(A=SceneTexelSize(), B=Max(A=@n32, B=@n35) {desc="If res smaller than target, calmp min value to 1"})
  ) {constA=2}
  n37 = Constant()
  n38 = AppendVector(A=ComponentMask(None=@n36) {desc=x, r=true}, B=@n37)
  n39 = Add(A=@n25, B=@n38) {desc=x+}
  n40 = AppendVector(A=@n37, B=ComponentMask(None=@n36) {desc=y, g=true}) {desc=y+}
  n41 = Add(A=@n25, B=@n40) {desc=y+}
  n42 = Constant() {r=-1}
  n43 = Add(A=@n25, B=Multiply(A=@n38, B=@n42)) {desc=x-}
  n44 = Add(A=@n25, B=Multiply(A=@n42, B=@n40)) {desc=y-}
  n45 = LinearInterpolate(
   A=ScalarParameter() {defaultValue=-1, parameterName=Divisor_Min}
   B=ScalarParameter() {defaultValue=-16, parameterName=Divisor_max}
   Alpha=@n22
  )
  n46 = LinearInterpolate(
   A=ScalarParameter() {defaultValue=-1, parameterName=MobileDivisor_Min}
   B=ScalarParameter() {defaultValue=-16, parameterName=MobileDivisor_max}
   Alpha=@n22
  )
  n47 = ComponentMask(None=MaterialFunctionCall!(
    Base=@n17.R
    BaseSampleBias=ShadingPathSwitch(Default=@n23, Mobile=@n24)
    Scale=ScalarParameter() {defaultValue=4, parameterName=Scale}
    Sample1=SceneDepth(UVs=@n39).R
    Sample2=SceneDepth(UVs=@n41).R
    Sample3=SceneDepth(UVs=@n43).R
    Sample4=SceneDepth(UVs=@n44).R
    Divisor=ShadingPathSwitch(Default=@n45, Mobile=@n46)
    UseCenteredLines
    UseRGB
   ) {materialFunction=/Game/Packages/jl_tests/combine4Samples}.Result) {r=true}
  n48 = ComponentMask(None=MaterialFunctionCall(
    Base=SceneTexture() {sceneTextureId=PPI_Roughness}.Color
    BaseSampleBias=Constant()
    Scale=Constant() {r=4}
    Sample1=SceneTexture(UVs=@n39) {sceneTextureId=PPI_Roughness}.Color
    Sample2=SceneTexture(UVs=@n41) {sceneTextureId=PPI_Roughness}.Color
    Sample3=SceneTexture(UVs=@n43) {sceneTextureId=PPI_Roughness}.Color
    Sample4=SceneTexture(UVs=@n44) {sceneTextureId=PPI_Roughness}.Color
    Divisor=ScalarParameter() {defaultValue=-0.04, group=RoughnessGBufferLines, parameterName=InnerLinesDivisor}
   ) {materialFunction=&combine4Samples}.Result) {r=true}
  n49 = StaticSwitchParameter(True=ShadingPathSwitch(Default=Max(A=@n47, B=@n48), Forward=@n47, Mobile=@n47), False=@n47) {group=RoughnessGBufferLines, parameterName=UseRoughnessGBufferOutlines}
  n50 = Multiply(A=Saturate(None=Multiply(A=@n19, B=@n49)), B=ScalarParameter() {defaultValue=1, group=Cartoon, parameterName=CRTNO_Outline-InnerAmount})
  n51 = Constant()
  n52 = StaticSwitchParameter(True=@n50, False=ShadingPathSwitch(Default=@n50, Forward=@n51, Mobile=@n51)) {group=Cartoon, parameterName=UseDepthBufferOutlinesOnMobile}
  n53 = Multiply(A=VertexNormalWS(), B=ScalarParameter() {defaultValue=0.5, parameterName=ScaleMesh})
  MaterialFunctionCall(In=SetMaterialAttributes!(
    MaterialAttributes
    Emissive Color=NamedRerouteDeclaration(None=@n7) {name=Emissive, nodeColor=<rgba 0.3125 0 1 1>}
    Opacity=NamedRerouteDeclaration(None=@n52) {name=Opacity, nodeColor=<rgba 0 1 0.0391 1>}
    World Position Offset=NamedRerouteDeclaration(None=@n53) {name=WPO, nodeColor=<rgba 1 0 0.7266 1>}
   ) {attributeSetTypes [#1, #2, #3]}) {materialFunction=&MF_TechArtOverrides}
 ]
}
