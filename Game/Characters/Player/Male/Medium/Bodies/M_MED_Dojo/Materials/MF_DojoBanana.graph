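# Material function for the Dojo player body's "Banana" that toggles vertex-color-tagged mesh pieces to follow facial poses.
# A priority If chain tests glabella, right brow, blink and squint pose parameters against thresholds to pick one vertex value.
# VertexColorMask compares that value to red plus green, adding a fixed red-channel mask.
# The inverted sum blends in a negated-world-position WPO on the input material, collapsing masked vertices to the world origin.
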MaterialFunction /Game/Characters/Player/Male/Medium/Bodies/M_MED_Dojo/Materials/MF_DojoBanana
graph {nodes [
  n1 = Constant() {r=1}
  n2 = VertexColor()
  n3 = NamedRerouteDeclaration(None=@n2.R) {name=Vert Red, nodeColor=<rgba 0.5469 1 0 1>}
  n4 = ScalarParameter() {group=Banana_Poses, parameterName=C_glabella_up_pose}
  n5 = ScalarParameter() {defaultValue=0.1, group=Banana, parameterName=Threshold_Glabella_up}
  n6 = Constant() {r=0.1}
  n7 = ScalarParameter() {defaultValue=0.2, group=Banana, parameterName="GlabellaDownValue(G)"}
  n8 = ScalarParameter() {defaultValue=0.4, group=Banana, parameterName="BananaBrowUpValue(R)"}
  n9 = ScalarParameter() {group=Banana_Poses, parameterName=R_blink_pose}
  n10 = ScalarParameter() {defaultValue=0.1, group=Banana, parameterName=Threshold_R_blink_pose}
  n11 = ScalarParameter() {defaultValue=0.6, group=Banana, parameterName="BananaBlinkPoseValue(R)"}
  n12 = ScalarParameter() {defaultValue=0.5, group=Banana, parameterName="BananaSquintPoseValue(R)"}
  n13 = If(
   A=ScalarParameter() {group=Banana_Poses, parameterName=R_squint_inner_pose}
   B=ScalarParameter() {defaultValue=0.375, group=Banana, parameterName=Threshold_squint_inner_pose}
   "A > B"=@n12
   "A == B"=@n12
   "A < B"=Constant() {r=0.3}
  )
  n14 = If(
   A=ScalarParameter() {group=Banana_Poses, parameterName=C_glabella_down_pose}
   B=ScalarParameter() {defaultValue=0.2, group=Banana, parameterName=Threshold_Glabella_down}
   "A > B"=@n7
   "A == B"=@n7
   "A < B"=If(
    A=ScalarParameter() {group=Banana_Poses, parameterName=R_brow_up_pose}
    B=ScalarParameter() {defaultValue=0.1, group=Banana, parameterName=Threshold_R_brow_up_pose}
    "A > B"=@n8
    "A == B"=@n8
    "A < B"=If(A=@n9, B=@n10, "A > B"=@n11, "A == B"=@n11, "A < B"=@n13)
   )
  )
  n15 = NamedRerouteDeclaration(None=@n2.G) {name=Vert Green, nodeColor=<rgba 0 1 0 1>}
  n16 = Add(
   A=MaterialFunctionCall!(Hide Element Toggle=@n1, VertexColorChannel=@n3, Vertex Value to hide=@n1) {materialFunction=/Game/Packages/Fortress_SharedMaterials/Base_Material_Functions/VertexColorMask}.Result
   B=MaterialFunctionCall(
    Hide Element Toggle=Constant() {r=1}
    VertexColorChannel=Add(A=@n3, B=@n15)
    Vertex Value to hide=If(A=@n4, B=@n5, "A > B"=@n6, "A == B"=@n6, "A < B"=@n14)
   ) {materialFunction=&VertexColorMask}.Result
  )
  n17 = WorldPosition()
  FunctionOutput(None=MaterialFunctionCall!(
    Material=FunctionInput() {inputName=Material, inputType=FunctionInput_MaterialAttributes}
    Mask=OneMinus(None=@n16)
    WorldPositionOffset=Multiply(A=@n17.XYZ) {constB=-1}
   ) {materialFunction=/Engine/Functions/MaterialLayerFunctions/MatLayerBlend_OverrideWorldPositionOffset}.Blended Material)
 ]}
