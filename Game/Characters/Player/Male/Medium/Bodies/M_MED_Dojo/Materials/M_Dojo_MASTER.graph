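# Master cloth-shaded character material for the Dojo medium male player body, usable on skeletal meshes with cloth, morph targets and mesh deformers.
# A shared PreFX character function builds material attributes, which pass through the MF_AnimeShading function for anime-style shading.
# A "Use Banana" static switch in the Skin FX group optionally routes that result through MF_DojoBanana.
# The shared PostFX character function then finishes the chain, also taking inputs from PreFX.
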Material /Game/Characters/Player/Male/Medium/Bodies/M_MED_Dojo/Materials/M_Dojo_MASTER
graph {
 properties {
  bAutomaticallySetUsageInEditor=false
  bUseMaterialAttributes=true
  bUsedWithClothing=true
  bUsedWithMeshDeformer=true
  bUsedWithMorphTargets=true
  bUsedWithSkeletalMesh=true
  pixelDepthOffsetMode=PDOM_Legacy
  shadingModel=MSM_Cloth
  usedShadingModels=Cloth
 }
 nodes [
  n1 = MaterialFunctionCall() {materialFunction=&BaseFN_CharacterFunction_PreFX}
  n2 = MaterialFunctionCall!(
   Material=@n1.Result
   LitDiffuseOverride
   SSC_AlphaOverride
   Tint Mask
   DF InkLine Visibility Mask
   DF Inkline Injection
   DF Inkline Color Override
   DF Inkline Color Override Mask
   CustomUVs
  ) {materialFunction=/Game/Packages/Fortress_SharedMaterials/Anime_Master_Material/MF_AnimeShading}
  MaterialFunctionCall(
   AO=@n1.AO
   Distance Blend=@n1.AO
   Fresnel=@n1.AO
   Gmap=@n1.AO
   Pre Skinned Local Position=@n1.AO
   Material=StaticSwitchParameter(
    True=MaterialFunctionCall!(Material=@n2.Result) {materialFunction=&/MF_DojoBanana}.Result
    False=@n2.Result
   ) {group=Skin FX, parameterName=Use Banana}
  ) {materialFunction=&BaseFN_CharacterFunction_PostFX}
 ]
}
